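# Physics asset for the Famine male two-handed pickaxe, giving physics bodies to its handle, a jaw and a dangling strap.
# A three-segment capsule strap chain swings and twists within limits, hanging from a small master body whose angular motion is free relative to the handle.
# The box jaw joins the handle with tight limits that allow mostly one swing axis, while a large capsule covers the weapon body.
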PhysicsAsset /BRCosmetics/Weapons/FORT_Melee/Pickaxe_Famin_Male/Mesh/Famine_Male_Axe_PhysicsAsset
bodies {
 Dyn_Jaw [{shapeType=Box, extentX=10, extentY=10, extentZ=10, center=<10 1.6437 0>}]
 Dyn_Master_Strap [{shapeType=Capsule, length=0.505, radius=0.505, rotation=<rot 0 -90 0>}]
 Dyn_Strap1 [{shapeType=Capsule, length=7, radius=1, center=<2 0 0>, rotation=<rot 90 0 0>}]
 Dyn_Strap2 [{shapeType=Capsule, length=7, radius=1, center=<2 0 0>, rotation=<rot 90 0 0>}]
 Dyn_Strap3 [{shapeType=Capsule, length=7, radius=1, center=<2 0 0>, rotation=<rot 90 0 0>}]
 Melee_TwoHanded_Handle [{shapeType=Capsule, length=1, radius=1}, {shapeType=Capsule, length=125.25, radius=30, center=<0 0 27.8013>}]
}
constraints [
 {bone1Name=Dyn_Strap2, bone2Name=Dyn_Strap1, swing1LimitDegrees=45, swing2LimitDegrees=45, twistLimitDegrees=45}
 {bone1Name=Dyn_Strap3, bone2Name=Dyn_Strap2, swing1LimitDegrees=60, swing2LimitDegrees=60, twistLimitDegrees=60}
 {bone1Name=Dyn_Strap1, bone2Name=Dyn_Master_Strap, swing1LimitDegrees=60, swing2LimitDegrees=60, twistLimitDegrees=20}
 {
  bone1Name=Dyn_Master_Strap
  bone2Name=Melee_TwoHanded_Handle
  swing1LimitDegrees=45
  swing1Motion=Free
  swing2LimitDegrees=45
  swing2Motion=Free
  twistLimitDegrees=45
  twistMotion=Free
 }
 {bone1Name=Dyn_Jaw, bone2Name=Melee_TwoHanded_Handle, swing1LimitDegrees=20, swing2LimitDegrees=1, twistLimitDegrees=1}
]
